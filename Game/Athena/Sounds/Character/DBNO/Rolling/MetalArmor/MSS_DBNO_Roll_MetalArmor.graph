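# Mono one-shot MetaSound preset that plays metal-armor clothing foley for rolling while down but not out (DBNO).
# Graph logic comes from parent MSS_DBNO_Roll, with inputs for crossfade, triggers and random pitch and volume.
# It fills the clothing sweetener array with reused MetalArmor jump waves, leaving the surface and second sweetener arrays empty.
# Playback uses third-person sprint footstep attenuation and the player foley sound class.
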MetaSoundSource /Game/Athena/Sounds/Character/DBNO/Rolling/MetalArmor/MSS_DBNO_Roll_MetalArmor
preset_parent=/Game/Athena/Sounds/Character/DBNO/Rolling/Settings/MSS_DBNO_Roll
inputs [
 Clothing Sweeteners {type=WaveAsset:Array}
 Clothing Sweeteners 1 {type=WaveAsset:Array}
 CrossfadeTime {type=Time}
 GrabTime {type=Trigger}
 Random.Pitch.Base {type=Float}
 Random.Pitch.Spread {type=Float}
 Random.Volume.Base {type=Float}
 Random.Volume.Cloth Sweetener Volume {type=Float}
 Random.Volume.Spread {type=Float}
 StartTime {type=Float}
 SurfaceSounds {type=WaveAsset:Array}
 Trigger Attack {type=Trigger}
 TriggerSwitch {type=Trigger}
 UE.Source.OnPlay {type=Trigger}
]
interfaces [UE.OutputFormat.Mono:1.0, UE.Source.OneShot:1.0, UE.Source:1.1]
outputs [
 PlayTime {type=Float}
 "UE.OutputFormat.Mono.Audio:0" {type=Audio}
 UE.Source.OneShot.OnFinished {type=Trigger}
]
pages [Default {input_defaults {
   CrossfadeTime=0.5
   GrabTime=null
   Random.Pitch.Base=0
   Random.Pitch.Spread=0.5
   Random.Volume.Base=6
   Random.Volume.Cloth Sweetener Volume=-3
   Random.Volume.Spread=0
   StartTime=0
   Trigger Attack=null
   TriggerSwitch=null
   UE.Source.OnPlay=null
   Clothing Sweeteners [
    /Game/Sounds/Fort_Player/Clothing/MetalArmor/Waves/Foley_Clothing_MetalArmor_Jump_06
    /Game/Sounds/Fort_Player/Clothing/MetalArmor/Waves/Foley_Clothing_MetalArmor_Jump_05
    /Game/Sounds/Fort_Player/Clothing/MetalArmor/Waves/Foley_Clothing_MetalArmor_Jump_04
    /Game/Sounds/Fort_Player/Clothing/MetalArmor/Waves/Foley_Clothing_MetalArmor_Jump_03
    /Game/Sounds/Fort_Player/Clothing/MetalArmor/Waves/Foley_Clothing_MetalArmor_Jump_02
    /Game/Sounds/Fort_Player/Clothing/MetalArmor/Waves/Foley_Clothing_MetalArmor_Jump_01
   ]
   Clothing Sweeteners 1 []
   SurfaceSounds []
  }}]
properties {
 assetImportData=&self:AssetImportData
 attenuationSettings=/Game/Sounds/Attenuation/Fort_Player_Footsteps_3P_Sprint_Athena_Att
 audioPropertiesSheet=/Game/Athena/Sounds/Character/DBNO/Rolling/Settings/APS_Foley_DBNORoll
 maxDistance=3584
 sampleRate=48000
 soundClassObject=&Fort_Player_Foley_Athena
 volume=1.1
}
